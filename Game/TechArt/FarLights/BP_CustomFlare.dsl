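; BP_CustomFlare is a configurable camera-facing light flare whose look is set entirely in the Construction Script; its event graph holds only disabled stubs.
; It creates dynamic materials on Satellite_Flare's two slots, pushing scale, color, brightness, depth fade, flicker, camera offset and mask variables.
; Slot 0 (MI_Flare_Cam_Facing) takes Close values, while slot 1 (MI_Flare_Cam_Facing_Close) takes Far values, opposite to its name.
; Both reuse the MID variable, leaving it pointing at slot 1.
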(event EventBeginPlay
  (|CallParentFunction) ; This node is disabled and will not be called.\nDrag off pins to build functionality.
)

(event Collision|EventActorBeginOverlap
  (params (OtherActor :type Actor%20Object%20Reference))
  (|CallParentFunction OtherActor) ; This node is disabled and will not be called.\nDrag off pins to build functionality.
)

(event EventTick
  (params (DeltaSeconds :type Float%20%28single-precision%29))
  (|CallParentFunction DeltaSeconds) ; This node is disabled and will not be called.\nDrag off pins to build functionality.
)

(fn ConstructionScript
  (bind _satellite_flare (Variables|Default|GetSatellite_Flare))
  (bind _texture_mask (Variables|Default|GetTexture%20Mask))
  (bind _camera_offset (Variables|Default|GetCamera%20Offset))
  (|CallParentFunction)
  (bind _returnvalue (Rendering|Material|CreateDynamicMaterialInstance _satellite_flare 0 "/Game/Environments/Helios/TechArt/FarLights/MI_Flare_Cam_Facing.MI_Flare_Cam_Facing" "None" :!class /Script/Engine.PrimitiveComponent))
  (bind _output_get (Variables|Default|SetMID _returnvalue))
  (Rendering|Material|SetScalarParameterValue _output_get "SpriteScale" (Variables|Default|GetScale%20Close) :!class /Script/Engine.MaterialInstanceDynamic)
  (Rendering|Material|SetVectorParameterValue _output_get "Color" (Variables|Default|GetColor%20Close) :!class /Script/Engine.MaterialInstanceDynamic)
  (Rendering|Material|SetScalarParameterValue _output_get "GlowFlickerScale" (Variables|Default|GetGlow%20Flicker%20Scale) :!class /Script/Engine.MaterialInstanceDynamic)
  (Rendering|Material|SetScalarParameterValue _output_get "DepthFade" (Variables|Default|GetDepth%20Fade%20Close) :!class /Script/Engine.MaterialInstanceDynamic)
  (Rendering|Material|SetScalarParameterValue _output_get "CenterBrightness" (Variables|Default|GetCenterBrightness%20Close) :!class /Script/Engine.MaterialInstanceDynamic)
  (Rendering|Material|SetScalarParameterValue _output_get "CoronaBrightness" (Variables|Default|GetCorona%20Brighness%20Close) :!class /Script/Engine.MaterialInstanceDynamic)
  (Rendering|Material|SetScalarParameterValue _output_get "EdgeBrightness" (Variables|Default|GetEdge%20Brightness%20Close) :!class /Script/Engine.MaterialInstanceDynamic)
  (Rendering|Material|SetScalarParameterValue _output_get "SpriteScaleVertical" (Variables|Default|GetScale%20Vertical) :!class /Script/Engine.MaterialInstanceDynamic)
  (Rendering|Material|SetScalarParameterValue _output_get "CameraOffset" _camera_offset :!class /Script/Engine.MaterialInstanceDynamic)
  (Rendering|Material|SetTextureParameterValue _output_get "FlareMask" _texture_mask)
  (bind _returnvalue_2 (Rendering|Material|CreateDynamicMaterialInstance _satellite_flare 1 "/Game/Environments/Helios/TechArt/FarLights/MI_Flare_Cam_Facing_Close.MI_Flare_Cam_Facing_Close" "None" :!class /Script/Engine.PrimitiveComponent))
  (bind _output_get_2 (Variables|Default|SetMID _returnvalue_2))
  (Rendering|Material|SetScalarParameterValue _output_get_2 "SpriteScaleMax" (Variables|Default|GetScale%20Far) :!class /Script/Engine.MaterialInstanceDynamic)
  (Rendering|Material|SetVectorParameterValue _output_get_2 "Color" (Variables|Default|GetColor%20Far) :!class /Script/Engine.MaterialInstanceDynamic)
  (Rendering|Material|SetScalarParameterValue _output_get_2 "GlowFlickerScale" (Variables|Default|GetGlow%20Flicker%20Scale) :!class /Script/Engine.MaterialInstanceDynamic)
  (Rendering|Material|SetScalarParameterValue _output_get_2 "DepthFade" (Variables|Default|GetDepth%20Fade%20Far) :!class /Script/Engine.MaterialInstanceDynamic)
  (Rendering|Material|SetScalarParameterValue _output_get_2 "CenterBrightness" (Variables|Default|GetCenter%20Brightness%20Far) :!class /Script/Engine.MaterialInstanceDynamic)
  (Rendering|Material|SetScalarParameterValue _output_get_2 "CoronaBrightness" (Variables|Default|GetCorona%20Brightness%20Far) :!class /Script/Engine.MaterialInstanceDynamic)
  (Rendering|Material|SetScalarParameterValue _output_get_2 "EdgeBrightness" (Variables|Default|GetEdge%20Brightness%20Far) :!class /Script/Engine.MaterialInstanceDynamic)
  (Rendering|Material|SetScalarParameterValue _output_get_2 "SpriteScaleVertical" (Variables|Default|GetScale%20Vertical) :!class /Script/Engine.MaterialInstanceDynamic)
  (Rendering|Material|SetScalarParameterValue _output_get_2 "CameraOffset" _camera_offset :!class /Script/Engine.MaterialInstanceDynamic)
  (Rendering|Material|SetTextureParameterValue _output_get_2 "FlareMask" _texture_mask))
